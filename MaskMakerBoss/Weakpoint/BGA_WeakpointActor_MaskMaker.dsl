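; Shootable mask weakpoint for the Mask Maker boss, extending a weakpoint base with aim assist and health reporting to the boss pawn.
; The server gets the pawn from its parent holder and scales health by playlist team size.
; While enabled, timers randomly ease it between top and bottom offsets.
; RepNotifies apply location, mesh-fitted collision and visibility with shield and break effects, while hits drive shield FX and a scale/emissive wobble.
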(event EventEndPlay
  (params (EndPlayReason :type EEndPlayReason%20Enum :default "Destroyed"))
  (AimAssist|UnregisterAimAssistTarget 0))

(event EventBeginPlay
  (bind _possiblerelativezoffset_hf (Variables|Hotfix|GetPossibleRelativeZOffset_Hf))
  (bind _returnvalue (ScalableFloat|GetValueatLevel _possiblerelativezoffset_hf))
  ;[0.07398 0.08974 0.15 0.5] BeginPlay
  (Networking|SwitchHasAuthority
    (:Authority
      (bind (_gamestate _playlist) (Class|FortAsyncActionCurrentPlaylistReady|PlaylistDataReadyAsync)
        (:OnReady
          (bind _asbp_mask_maker_weakpoint_holder (Utilities|Casting|CastToBP_MaskMakerWeakpointHolder (Actor|GetParentActor)))
          (Utilities|IsValid (Variables|/MaskMakerBoss/Weakpoint/BP_MaskMakerWeakpointHolder.BP_MaskMakerWeakpointHolder_C|GetMaskMakerPawn _asbp_mask_maker_weakpoint_holder)
            (:Is%20Valid
              (Variables|Default|SetMaskMakerPawn (Variables|/MaskMakerBoss/Weakpoint/BP_MaskMakerWeakpointHolder.BP_MaskMakerWeakpointHolder_C|GetMaskMakerPawn _asbp_mask_maker_weakpoint_holder))
              (Variables|Default|SetMaxTeamSize (Math|Integer|Clamp(Integer) (Variables|/Script/FortniteGame.FortPlaylist|GetMaxTeamSize _playlist) 1 4))
              (Health|SetMaxHealthDontCreateASC 999999999.0)
              (CallFunction|ResetWeakpointHealth)
              (Variables|SetTopLocation_X (Math|Vector|MakeVector 0.0 0.0 _returnvalue))
              (Variables|SetBottomLocation_X (Math|Vector|MakeVector 0.0 0.0 (* _returnvalue -1.0))))))))
    (:Remote
      (Variables|Internal|SetbCollisionEnabled (Variables|Internal|GetbIsEnabled))
      (CallFunction|UpdateMeshandCollisionSize)
      (Rendering|SetVisibility (Variables|Default|GetWeakpointMesh) (Variables|Default|GetWeakpointVisible) false)
      (AimAssist|RegisterAimAssistTarget 0 (Variables|Default|GetWeakpointCollisionSphereMM)))))

(event EventOnDamagePlayEffects
  (params (Damage :type Float%20%28single-precision%29) (DamageTags :type Gameplay%20Tag%20Container%20Structure%20%28by%20ref%29) (Momentum :type Vector :default "0, 0, 0") (HitInfo :type Hit%20Result%20Structure%20%28by%20ref%29) (InstigatedBy :type Fort%20Pawn%20Object%20Reference) (DamageCauser :type Actor%20Object%20Reference) (EffectContext :type Gameplay%20Effect%20Context%20Handle%20Structure))
  (bind _weakpoint_currenthealth (Variables|Internal|GetWeakpoint_CurrentHealth))
  (bind _returnvalue (Math|Float|MapRangeClamped _weakpoint_currenthealth 0.0 2e+02 0.0 1.0))
  (bind _ns_maskmaker_maskshield (Variables|Default|GetNS_MaskMaker_MaskShield))
  (bind _weakpointmesh (Variables|Default|GetWeakpointMesh))
  ;[0.6198 0.1125 0.3398 0.5] Update shield FX
  (|CallParentFunction Damage DamageTags Momentum HitInfo InstigatedBy DamageCauser EffectContext)
  (Utilities|FlowControl|Sequence
    (:then_0
      (Development|PrintString (Utilities|String|Append (Utilities|String|ToString(Float) _weakpoint_currenthealth) (Utilities|String|ToString(Float) _returnvalue)) false false "(R=1.000000,G=0.000000,B=0.070813,A=1.000000)" 22.0)
      (Effects|Components|ParticleSystem|SetFloatParameter _ns_maskmaker_maskshield "NormalizedMaskHealth" _returnvalue)
      (Effects|Components|ParticleSystem|SetBoolParameter _ns_maskmaker_maskshield "PlayHitFX" false)
      (Utilities|FlowControl|DelayUntilNextTick)
      (Effects|Components|ParticleSystem|SetBoolParameter _ns_maskmaker_maskshield "PlayHitFX" true)
      (Utilities|FlowControl|DelayUntilNextTick)
      (Effects|Components|ParticleSystem|SetBoolParameter _ns_maskmaker_maskshield "PlayHitFX" false))
    (:then_1
      ;[0.5312 0.2693 0.1346 0.5] WobbleAnimOnShoot
      (bind (_direction _wobblecurve) (Timeline|MaskWobble PlayFromStart)
        (:Update
          (Transformation|SetRelativeScale3D _weakpointmesh (Math|Conversions|ToVector(Float) (* (Math|Float|Lerp 1.0 1.3 _wobblecurve) (ScalableFloat|GetValueatLevel (Variables|Hotfix|GetWeakpointScale_Hf)))))
          (Rendering|Material|SetScalarParameterValueonMaterials _weakpointmesh "EmissiveBrightness" (Math|Float|Lerp 0.0 5.0 _wobblecurve)))))))

(event EventOnDeathPlayEffects
  (params (Damage :type Float%20%28single-precision%29) (DamageTags :type Gameplay%20Tag%20Container%20Structure%20%28by%20ref%29) (Momentum :type Vector :default "0, 0, 0") (HitInfo :type Hit%20Result%20Structure%20%28by%20ref%29) (InstigatedBy :type Fort%20Pawn%20Object%20Reference) (DamageCauser :type Actor%20Object%20Reference) (EffectContext :type Gameplay%20Effect%20Context%20Handle%20Structure))
  (|CallParentFunction Damage DamageTags Momentum HitInfo InstigatedBy DamageCauser EffectContext))

(event Custom|SetMeshFromSoftRef
  (params (SM%20Soft%20Reference :type Static%20Mesh%20Soft%20Object%20Reference))
  (bind _object (Utilities|AsyncLoadAsset SM%20Soft%20Reference)
    (:Completed
      (bind _asstatic_mesh (Utilities|Casting|CastToStaticMesh _object))
      (Variables|Default|SetWeakpointMeshVar _asstatic_mesh))))

(event Custom|TryToOffsetMask
  (if (Math|Random|RandomBool)
    (Variables|Default|SetOffsetMoveAlpha 0.0)
    (bind _returnvalue (Utilities|Time|SetTimerbyEvent (EventDispatchers|CreateEvent :!event Custom|MaskOffsetMove) (ScalableFloat|GetValueatLevel (Variables|Hotfix|GetMoveOffsetPeriod_Hf)) true true))
    (Variables|Default|SetMaskOffsetMoveTHandle _returnvalue)))

(event Custom|MaskOffsetMove
  (bind _output_get (Variables|Default|SetOffsetMoveAlpha (+ (Variables|Default|GetOffsetMoveAlpha) (/ (ScalableFloat|GetValueatLevel (Variables|Hotfix|GetMoveOffsetPeriod_Hf)) (ScalableFloat|GetValueatLevel (Variables|Hotfix|GetMoveOffsetDuration_Hf))))))
  (Variables|Default|SetCurrentRelativeLocation (Math|Interpolation|Ease "EaseOut" _output_get (Math|Vector|SelectVector (Variables|Default|GetTopLocation) (Variables|Default|GetBottomLocation) (Variables|Default|GetIsAtTopLocation)) (Math|Vector|SelectVector (Variables|Default|GetBottomLocation) (Variables|Default|GetTopLocation) (Variables|Default|GetIsAtTopLocation)) 2.0))
  (if (>= (Variables|Default|GetOffsetMoveAlpha) 1.0)
    (Utilities|Time|ClearandInvalidateTimerbyHandle (Variables|Default|GetMaskOffsetMoveTHandle))
    (Variables|Default|SetIsAtTopLocation (not (Variables|Default|GetIsAtTopLocation)))))

(fn DisableWeakpoint
  (|CallParentFunction)
  (Variables|Default|SetWeakpointVisible false)
  (Utilities|Time|ClearandInvalidateTimerbyHandle (Variables|Default|GetTryMaskOffsetTHandle))
  (Utilities|Time|ClearandInvalidateTimerbyHandle (Variables|Default|GetMaskOffsetMoveTHandle)))

(fn EnableWeakpoint
  (|CallParentFunction)
  (Variables|Default|SetWeakpointVisible true)
  (Variables|Default|SetIsAtTopLocation (Math|Random|RandomBool))
  (Variables|Default|SetCurrentRelativeLocation (Math|Vector|SelectVector (Variables|Default|GetTopLocation) (Variables|Default|GetBottomLocation) (Variables|Default|GetIsAtTopLocation)))
  (bind _returnvalue (Utilities|Time|SetTimerbyEvent (EventDispatchers|CreateEvent :!event TryToOffsetMask) (Math|Random|RandomFloatinRange (ScalableFloat|GetValueatLevel (Variables|Hotfix|GetMinOffsetMoveDelay_Hf)) (ScalableFloat|GetValueatLevel (Variables|Hotfix|GetMaxOffsetMoveDelay_Hf))) true true))
  (Variables|Default|SetTryMaskOffsetTHandle _returnvalue))

(fn GetWeakpointCurrentHealth
  (return (Variables|Internal|GetWeakpoint_CurrentHealth)))

(fn IsValidAutoFireTarget
  (params (Attacker :type Actor%20Object%20Reference))
  (return (Variables|Internal|GetbIsEnabled)))

(fn NotifyParent
  (params (Damage :type Float%20%28double-precision%29) (InstigatedBy :type Controller%20Object%20Reference) (DamageCauser :type Actor%20Object%20Reference))
  (bind _weakpoint_currenthealth (Variables|Internal|GetWeakpoint_CurrentHealth))
  (Class|BPPlayerPawnCherrySmokeMaskMakerBoss|MaskWeakpointHealthUpdate (Variables|Default|GetMaskMakerPawn) _weakpoint_currenthealth self)
  (Development|PrintString (Utilities|String|ToString(Float) _weakpoint_currenthealth) false false "(R=1.000000,G=0.126923,B=0.490199,A=1.000000)" 22.0))

(fn OnRep_CurrentRelativeLocation
  (Transformation|SetActorRelativeLocation (Variables|Default|GetCurrentRelativeLocation)))

(fn OnRep_WeakpointMeshVar
  (CallFunction|UpdateMeshandCollisionSize))

(fn OnRep_WeakpointVisible
  (bind _returnvalue (Transformation|GetActorLocation))
  (bind _weakpointvisible (Variables|Default|GetWeakpointVisible))
  (bind _ns_maskmaker_maskshield (Variables|Default|GetNS_MaskMaker_MaskShield))
  (Networking|CanExecuteCosmeticEvents
    (:True
      (Rendering|SetVisibility (Variables|Default|GetWeakpointMesh) _weakpointvisible false)
      (if _weakpointvisible
        (Components|Activation|Activate _ns_maskmaker_maskshield false)
        (Effects|Components|ParticleSystem|SetFloatParameter _ns_maskmaker_maskshield "NormalizedMaskHealth" 1.0)
        (else
          (Components|Activation|Deactivate _ns_maskmaker_maskshield)
          (Development|PrintString "MaskBreak" true true "(R=1.000000,G=0.000000,B=0.070813,A=1.000000)" 22.0)
          (Effects|Components|ParticleSystem|SpawnFXAtLocation "/MaskMakerBoss/VFX/NS_MaskMaker_MaskBreak.NS_MaskMaker_MaskBreak" _returnvalue)
          (Audio|PlaySoundatLocation "/MaskMakerBoss/Assets/Audio/SC_MaskMakerBoss_MaskDestroy.SC_MaskMakerBoss_MaskDestroy" _returnvalue))))))

(fn ResetWeakpointHealth
  (bind _maxteamsize (Variables|Default|GetMaxTeamSize))
  (bind _returnvalue (Math|Conversions|ToFloat(Integer) _maxteamsize))
  (bind _weakpointplaylistmaxhealth_hf (Variables|Hotfix|GetWeakpointPlaylistMaxHealth_Hf))
  (bind _returnvalue_2 (ScalableFloat|GetValueatLevel _weakpointplaylistmaxhealth_hf _returnvalue))
  (Development|PrintString (Utilities|String|Append "@@@ ResetWeakpointHealth [BGA_WeakpointActor_Base]: CALLED for " (Utilities|GetDisplayName self) "!") false (Variables|Debug|GetbDebugWeakpoint))
  (if (> _returnvalue_2 0)
    (Variables|Internal|SetWeakpoint_CurrentHealth _returnvalue_2)))

(fn UpdateMeshAndCollisionSize
  (bind (_origin _boxextent _sphereradius) (Math|BoxSphereBounds|BreakBoxSphereBounds (StaticMesh|GetBounds (Variables|Default|GetWeakpointMeshVar))))
  (bind _relativescale3d_x (Math|Vector|BreakVector (Variables|/Script/Engine.SceneComponent|GetRelativeScale3D_X (Variables|Default|GetWeakpointMesh))))
  (bind _weakpointmeshvar (Variables|Default|GetWeakpointMeshVar))
  (Utilities|IsValid _weakpointmeshvar
    (:Is%20Valid
      (Components|StaticMesh|SetStaticMesh (Variables|Default|GetWeakpointMesh) _weakpointmeshvar)
      (Components|Sphere|SetSphereRadius (Variables|Default|GetWeakpointCollisionSphereMM) (+ (* _sphereradius _relativescale3d_x) (ScalableFloat|GetValueatLevel (Variables|Hotfix|GetCollisionAdditionalRadius_Hf))) false))))

(fn UpdateWeakpointCollision
  (params (bEnableCollision :type Boolean) (CollisionKind :type Enum_WeakpointActor_CollisionKind%20Enum :default "NewEnumerator0") (CollisionComponent :type Primitive%20Component%20Object%20Reference))
  (|CallParentFunction bEnableCollision (Variables|Internal|GetWeakpointCollisionKind) (Variables|Default|GetWeakpointCollisionSphereMM)))

(fn ConstructionScript
  (|CallParentFunction)
  (Transformation|SetRelativeScale3D (Variables|Default|GetWeakpointMesh) (Math|Conversions|ToVector(Float) (ScalableFloat|GetValueatLevel (Variables|Hotfix|GetWeakpointScale_Hf)))))
