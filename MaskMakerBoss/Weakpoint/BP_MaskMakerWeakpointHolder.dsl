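; Weakpoint holder for the Mask Maker boss that carries four mask weakpoints orbiting the boss pawn.
; Construction spaces them evenly around a height-offset orbit point using hotfixable values.
; On a hotfixable tick interval, the server follows the instigator boss and advances yaw faster as health drops, with speed limits looked up by max team size.
; OnRep functions apply replicated location and rotation, and the holder destroys itself when the boss dies.
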(event EventBeginPlay
  (Networking|SwitchHasAuthority
    (:Authority
      (bind _asbp_player_pawn_cherry_smoke_mask_maker_boss (Utilities|Casting|CastToBP_PlayerPawn_CherrySmoke_MaskMakerBoss (Game|GetInstigator)))
      (bind _output_get (Variables|Default|SetMaskMakerPawn _asbp_player_pawn_cherry_smoke_mask_maker_boss))
      (EventDispatchers|BindEventtoOnDied _output_get (EventDispatchers|CreateEvent :!event Custom|MaskMakerEliminated))
      (Actor|Tick|SetActorTickInterval (ScalableFloat|GetValueatLevel (Variables|Hotfix|GetWeakpointOrbitPeriod_Hf))))))

(event EventTick
  (params (DeltaSeconds :type Float%20%28single-precision%29))
  (bind _maskmakerpawn (Variables|Default|GetMaskMakerPawn))
  (Networking|SwitchHasAuthority
    (:Authority
      (Utilities|IsValid _maskmakerpawn
        (:Is%20Valid
          (Variables|Default|SetHolderLocation (Transformation|GetActorLocation _maskmakerpawn))
          (Variables|Default|SetHolderZRotation (+ (Variables|Default|GetHolderZRotation) (Math|Float|Lerp (ScalableFloat|GetValueatLevel (Variables|Hotfix|GetWeakpointMaxOrbitSpeed_Hf) (Math|Conversions|ToFloat(Integer) (Variables|/MaskMakerBoss/Pawn/BP_PlayerPawn_CherrySmoke_MaskMakerBoss.BP_PlayerPawn_CherrySmoke_MaskMakerBoss_C|GetMaxTeamSize (Variables|Default|GetMaskMakerPawn)))) (ScalableFloat|GetValueatLevel (Variables|Hotfix|GetWeakpointMinOrbitSpeed_Hf) (Math|Conversions|ToFloat(Integer) (Variables|/MaskMakerBoss/Pawn/BP_PlayerPawn_CherrySmoke_MaskMakerBoss.BP_PlayerPawn_CherrySmoke_MaskMakerBoss_C|GetMaxTeamSize (Variables|Default|GetMaskMakerPawn)))) (/ (FortPawn|GetHealth (Variables|Default|GetMaskMakerPawn)) (FortPawn|GetMaxHealth (Variables|Default|GetMaskMakerPawn)))))))))))

(event Custom|MaskMakerEliminated
  (params (DamagedActor :type Actor%20Object%20Reference) (Damage :type Float%20%28single-precision%29) (InstigatedBy :type Controller%20Object%20Reference) (DamageCauser :type Actor%20Object%20Reference) (HitLocation :type Vector) (FHitComponent :type Primitive%20Component%20Object%20Reference) (BoneName :type Name) (Momentum :type Vector))
  (Actor|DestroyActor))

(fn OnRep_HolderLocation
  (Transformation|SetActorLocation (Variables|Default|GetHolderLocation)))

(fn OnRep_HolderZRotation
  (Transformation|SetActorRotation 0 (Math|Rotator|MakeRotator 0.0 0.0 (Variables|Default|GetHolderZRotation)) false))

(fn ConstructionScript
  (for _index (range 4)
    (Transformation|SetRelativeLocation (Utilities|Select :Index _index :Option%200 (Variables|Default|GetMaskWeakpoint1) :Option%201 (Variables|Default|GetMaskWeakpoint2) :Option%202 (Variables|Default|GetMaskWeakpoint3) :Option%203 (Variables|Default|GetMaskWeakpoint4)) (Utilities|Select :Index _index :Option%200 (Math|Vector|MakeVector (ScalableFloat|GetValueatLevel (Variables|Hotfix|GetWeakpointDistFromRoot_Hf))) :Option%201 (Math|Vector|MakeVector 0.0 (ScalableFloat|GetValueatLevel (Variables|Hotfix|GetWeakpointDistFromRoot_Hf))) :Option%202 (Math|Vector|MakeVector (* (ScalableFloat|GetValueatLevel (Variables|Hotfix|GetWeakpointDistFromRoot_Hf)) -1.0)) :Option%203 (Math|Vector|MakeVector 0.0 (* (ScalableFloat|GetValueatLevel (Variables|Hotfix|GetWeakpointDistFromRoot_Hf)) -1.0))) false false)
    (Transformation|SetRelativeLocation (Variables|Default|GetWeakpointOrbitPoint) (Math|Vector|MakeVector 0.0 0.0 (ScalableFloat|GetValueatLevel (Variables|Hotfix|GetWeakpointsZOffset_Hf))) false false)))
